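# ClearMaterial is an unlit translucent material in the BRCosmetics unique materials package, named as a clear surface.
# Its graph holds a single Constant node at its default value, with no textures, parameters or visible wiring to material outputs.
# It is flagged for use on skeletal meshes, instanced static meshes and mesh deformers.
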Material /BRCosmetics/Packages/Fortnite_UniqueMaterials_1/ClearMaterial
graph {
 properties {
  bUsedWithInstancedStaticMeshes=true
  bUsedWithMeshDeformer=true
  bUsedWithSkeletalMesh=true
  blendMode=BLEND_Translucent
  pixelDepthOffsetMode=PDOM_Legacy
  shadingModel=MSM_Unlit
  usedShadingModels=Unlit
 }
 nodes [Constant()]
}
